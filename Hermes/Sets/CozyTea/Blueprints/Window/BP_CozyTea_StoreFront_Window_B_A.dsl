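; Construction script of a storefront window Blueprint in the CozyTea set.
; It runs the parent construction script, then calls SpawnAddonMesh from AddonMeshOptions.
; That call receives the Spawn Addon Mesh and Addon Mesh variables along with the actor itself, so the window can carry an optional add-on mesh.
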(fn ConstructionScript
  (|CallParentFunction)
  (AddonMeshOptions|SpawnAddonMesh (Variables|Default|GetSpawn%20Addon%20Mesh) (Variables|Default|GetAddon%20Mesh) self))
