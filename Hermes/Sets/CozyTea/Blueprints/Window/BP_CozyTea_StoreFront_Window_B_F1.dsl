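; Blueprint for a storefront window piece in the CozyTea set, named as variant B, F1.
; Its construction script is empty, and the file shows no other functions, variables or event logic.
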(fn ConstructionScript)
